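; Actor Blueprint that controls the audio mix for Risky Reels video playback.
; On BeginPlay it finds Athena_Prop_Video_Stream_Trigger1 actors, checks validity with a short delay on the invalid branch, and stores one as its stream trigger.
; It binds to that trigger's video started and stopped dispatchers, playing and stopping the MixApplicationAudio component accordingly.
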(event EventBeginPlay
  (bind _outactors (Actor|GetAllActorsOfClass "/Game/Athena/BuildingActors/Props/Building/ActorBlueprints/Containers/Athena_Prop_Video_Stream_Trigger1.Athena_Prop_Video_Stream_Trigger1_C"))
  (if (> (Utilities|Array|Length _outactors) 0)
    (Utilities|IsValid (Utilities|Array|Get(acopy) _outactors)
      (:Is%20Valid
        _)
      (:Is%20Not%20Valid
        (Utilities|FlowControl|Delay 0.2)))
    (bind _output_get (Variables|Default|SetStreamTriggerBP (Utilities|Array|Get(acopy) _outactors)))
    (Default|BindEventtoVideoStartedEventDispatcher _output_get (delegate Custom|Video%20Started))
    (Default|BindEventtoVideoStoppedEventDispatcher _output_get (delegate Custom|Video%20Stopped))))

(event Custom|Video%20Started
  (params (FirstStart :type Boolean))
  (Audio|Components|Audio|Play (Variables|Default|GetMixApplicationAudio) 0.0))

(event Custom|Video%20Stopped
  (Audio|Components|Audio|Stop (Variables|Default|GetMixApplicationAudio)))

(fn ConstructionScript)
