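# Sound cue for the Risky Reels event that plays looping silence, apparently as a carrier to apply a mix rather than to be heard.
# A single looping Wave Player feeds a five-second silence wave straight to the output.
# The cue uses the Risky Reels screen attenuation and the Fort_Live_InWorld_Risky_Active sound class, so that class is active in-world around the screen while the cue plays.
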SoundCue /Game/Athena/Sounds/Events/RiskyReels/RiskyReels_MixApplication_Cue
root=WavePlayer_0
nodes [{id=WavePlayer_0, properties {bLooping=true, soundWaveAssetPtr=/Game/Sounds/Fort_Test_Sounds/5_seconds_silence}}]
properties {
 attenuationSettings=/Game/Sounds/Attenuation/Fort_RiskyReelsScreen_Att
 maxDistance=20000
 soundClassObject=/Game/Sounds/FortSoundClassesAndModes/Fort_Live_InWorld_Risky_Active
}
